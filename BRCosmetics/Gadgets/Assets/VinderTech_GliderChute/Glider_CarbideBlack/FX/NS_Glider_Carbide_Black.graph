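# Niagara effect for the Carbide Black glider cosmetic, adding wing trails, red flares and lights to the glider's skeletal mesh.
# Four ribbon AnimTrail emitters span socket pairs, while the sprite Flare and the Lights emitter use spawn groups to index four sockets and offsets.
# User.bisFullyDeployed and a User.ForwardDot threshold together gate trails, flare spawn rate and light enable.
# Lights are burst-spawned and switched by that condition, not by spawning.
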NiagaraSystem /BRCosmetics/Gadgets/Assets/VinderTech_GliderChute/Glider_CarbideBlack/FX/NS_Glider_Carbide_Black
dependencies {
 usedDataInterfaces [
  &NiagaraDataInterfaceArrayInt32
  &NiagaraDataInterfaceArrayFloat3
  &NiagaraDataInterfaceColorCurve
  &NiagaraDataInterfaceCurve
 ]
 usedDynamicInputs []
 usedModules [&SystemState]
}
emitters {
 Flare {
  emitterUpdateScript [
   &EmitterState()
   &SpawnRate(
    SpawnRate=&Multiply_Float(
     A=6
     B=&MakeCustomFloatFromBool(
      Bool=&Boolean_AndOperation(
       Bool1=$User.bisFullyDeployed
       Bool2=&SetBoolByFloatComparison(A=$User.ForwardDot, B=0.1, Comparison Type=::NewEnumerator2)
      )
      True Float=1
      False Float=0
     )
    )
    Use Spawn Probability=false
   )
   &SpawnRate(
    SpawnGroup=1
    SpawnRate=&Multiply_Float(
     A=6
     B=&MakeCustomFloatFromBool(
      Bool=&Boolean_AndOperation(
       Bool1=$User.bisFullyDeployed
       Bool2=&SetBoolByFloatComparison(A=$User.ForwardDot, B=0.1, Comparison Type=::NewEnumerator2)
      )
      True Float=1
      False Float=0
     )
    )
    Use Spawn Probability=false
   ) {module=SpawnRate001}
   &SpawnRate(
    SpawnGroup=2
    SpawnRate=&Multiply_Float(
     A=6
     B=&MakeCustomFloatFromBool(
      Bool=&Boolean_AndOperation(
       Bool1=$User.bisFullyDeployed
       Bool2=&SetBoolByFloatComparison(A=$User.ForwardDot, B=0.1, Comparison Type=::NewEnumerator2)
      )
      True Float=1
      False Float=0
     )
    )
    Use Spawn Probability=false
   ) {module=SpawnRate002}
   &SpawnRate(
    SpawnGroup=3
    SpawnRate=&Multiply_Float(
     A=6
     B=&MakeCustomFloatFromBool(
      Bool=&Boolean_AndOperation(
       Bool1=$User.bisFullyDeployed
       Bool2=&SetBoolByFloatComparison(A=$User.ForwardDot, B=0.1, Comparison Type=::NewEnumerator2)
      )
      True Float=1
      False Float=0
     )
    )
    Use Spawn Probability=false
   ) {module=SpawnRate003}
  ]
  particleSpawnScript [&InitializeParticle(
    Color Minimum=<rgba 0.8 0.8 0.8 1>
    Color Mode=::NewEnumerator2
    Lifetime Max=1
    Lifetime Min=0.5
    Lifetime Mode=::NewEnumerator1
    Sprite Size Mode=::NewEnumerator3
    Uniform Sprite Size=90
    UsePositionOffset=false
    Write Material Random=true
   )]
  particleUpdateScript [
   &ParticleState()
   &SkeletalMeshLocation(
    Bone / Socket Index=&SelectIntFromArray(Direct Array Index=&SpawnGroup(), Int Selection Array=DataInterface() {propertyValues {IntData [0,4,3,7]}})
    MaskSpawnGroup=false
    Mesh Sampling Type=::NewEnumerator5
    Sampled Position Offset=&SelectVectorFromArray(
     Direct Array Index=&SpawnGroup()
     Vector Selection Array=DataInterface() {propertyValues {FloatData [<0 0 0>, <0 0 0>, <10 15 0>, <15 11 0>]}}
    )
    Skeletal Mesh=$User.Glider
    Socket Sampling Mode=::NewEnumerator5
   )
   &CameraOffset(Camera Offset Amount=&RandomRangeFloat(Minimum=10, Maximum=40, Evaluation Type=::NewEnumerator1))
   &ScaleColor(
    Scale Alpha=&FloatFromCurve(FloatCurve=DataInterface() {propertyValues {Curve {keys [
         {arriveTangent=33.8119, interpMode=RCIM_Cubic, leaveTangent=33.8119, tangentMode=RCTM_User, time=0, value=0}
         {arriveTangent=0.6492, interpMode=RCIM_Cubic, leaveTangent=0.6492, time=0.0608, value=1.0023}
         {arriveTangent=-0.8008, interpMode=RCIM_Cubic, leaveTangent=-0.8008, tangentMode=RCTM_User, time=0.2803, value=0.182}
         <at 1 0 RCIM_Cubic>
        ]}}})
    Scale RGB=&MakeVectorFromLinearColorRGB(Color=&ColorFromCurve(ColorCurve=DataInterface() {propertyValues {
        AlphaCurve {keys []}
        BlueCurve {keys [<at 0 0.1443 RCIM_Cubic>,<at 0.894 0.0954 RCIM_Cubic>]}
        GreenCurve {keys [<at 0 0.1443 RCIM_Cubic>,<at 0.894 0.0954 RCIM_Cubic>]}
        RedCurve {keys [<at 0 1.5 RCIM_Cubic>,<at 0.894 0.5473 RCIM_Cubic>]}
       }}))
    ScaleA=true
    ScaleRGB=true
   )
   &ScaleSpriteSize(Uniform Curve Sprite Scale=DataInterface() {propertyValues {Curve {keys [
        <at 0 0.6 RCIM_Cubic>
        {arriveTangent=0.9213, interpMode=RCIM_Cubic, leaveTangent=0.9213, time=0.1774, value=1.172}
        {arriveTangent=0.7635, interpMode=RCIM_Cubic, leaveTangent=0.7635, time=0.4656, value=1.029}
        <at 1 1.8 RCIM_Cubic>
       ]}}})
  ]
  rendererClasses [&NiagaraSpriteRendererProperties]
  renderers [{rendererClass=&NiagaraSpriteRendererProperties}]
 }
 Lights {
  emitterUpdateScript [
   &EmitterState(
    Inactive Response=::NewEnumerator1
    Life Cycle Mode=::NewEnumerator1
    Loop Behavior=::NewEnumerator1
    Loop Duration Mode=::NewEnumerator1
    UseLoopDelay=false
   )
   &SpawnBurst_Instantaneous(Use Spawn Probability=false, UseLoopCountLimit=false)
   &SpawnBurst_Instantaneous(Spawn Group=1, Use Spawn Probability=false, UseLoopCountLimit=false) {module=SpawnBurst_Instantaneous001}
   &SpawnBurst_Instantaneous(Spawn Group=2, Use Spawn Probability=false, UseLoopCountLimit=false) {module=SpawnBurst_Instantaneous002}
   &SpawnBurst_Instantaneous(Spawn Group=3, Use Spawn Probability=false, UseLoopCountLimit=false) {module=SpawnBurst_Instantaneous003}
  ]
  particleSpawnScript [&InitializeParticle(
    Color Maximum=<rgba 1 1 1 0.4>
    Color Minimum=<rgba 0.8 0.8 0.8 0.4>
    Color Mode=::NewEnumerator2
    Lifetime=1.5
    Sprite Size Mode=::NewEnumerator3
    Uniform Sprite Size=90
    UsePositionOffset=false
    Write Material Random=true
   )]
  particleUpdateScript [
   &ParticleState(Kill Particles When Lifetime Has Elapsed=false, Let Infinitely Lived Particles Die When Emitter Deactivates=true)
   &SkeletalMeshLocation(
    Bone / Socket Index=&SelectIntFromArray(Direct Array Index=&SpawnGroup(), Int Selection Array=DataInterface() {propertyValues {IntData [0,4,3,7]}})
    MaskSpawnGroup=false
    Mesh Sampling Type=::NewEnumerator5
    Sampled Position Offset=&SelectVectorFromArray(
     Direct Array Index=&SpawnGroup()
     Vector Selection Array=DataInterface() {propertyValues {FloatData [<0 0 0>, <0 0 0>, <10 15 0>, <15 11 0>]}}
    )
    Skeletal Mesh=$User.Glider
    Socket Sampling Mode=::NewEnumerator5
   )
   &ScaleColor(
    Scale Alpha=&FloatFromCurve(FloatCurve=DataInterface() {propertyValues {Curve {keys [
         {arriveTangent=33.8119, interpMode=RCIM_Cubic, leaveTangent=33.8119, tangentMode=RCTM_User, time=0, value=0}
         {arriveTangent=0.6492, interpMode=RCIM_Cubic, leaveTangent=0.6492, time=0.0608, value=1.0023}
         {arriveTangent=-0.8008, interpMode=RCIM_Cubic, leaveTangent=-0.8008, tangentMode=RCTM_User, time=0.2803, value=0.182}
         <at 1 0 RCIM_Cubic>
        ]}}})
    Scale RGB=&MakeVectorFromLinearColorRGB(Color=&ColorFromCurve(ColorCurve=DataInterface() {propertyValues {
        AlphaCurve {keys []}
        BlueCurve {keys [<at 0 0.1443 RCIM_Cubic>,<at 0.894 0.0954 RCIM_Cubic>]}
        GreenCurve {keys [<at 0 0.1443 RCIM_Cubic>,<at 0.894 0.0954 RCIM_Cubic>]}
        RedCurve {keys [<at 0 1.5 RCIM_Cubic>,<at 0.894 0.5473 RCIM_Cubic>]}
       }}))
    ScaleA=true
    ScaleRGB=true
   )
   &FN_Light_Attributes(
    Color=$Particles.Color
    Create Color Attribute=true
    Create Exponent Attribute=true
    Create Radius Attribute=true
    Enable Particle Light Enable=true
    Exponent=2
    Light Enabled:NiagaraBool=&Boolean_AndOperation(
     Bool1=$User.bisFullyDeployed
     Bool2=&SetBoolByFloatComparison(A=$User.ForwardDot, B=0.1, Comparison Type=::NewEnumerator2)
    )
    Radius=75
    Scale Color=10
   )
  ]
  rendererClasses [&NiagaraLightRendererProperties]
  renderers [{rendererClass=&NiagaraLightRendererProperties, rendererIndex=0}]
 }
 Wyeth_AnimTrails {
  emitterUpdateScript [
   &EmitterState()
   &self:SpawnAnimTrail(
    AnimTrail is Active:NiagaraBool=&Boolean_AndOperation(
     Bool1=$User.bisFullyDeployed
     Bool2=&SetBoolByFloatComparison(A=$User.ForwardDot, B=0.1, Comparison Type=::NewEnumerator2)
    )
    Mesh:NiagaraDataInterfaceSkeletalMesh=$User.Glider
    Socket B=1
   ) {module=SpawnAnimTrail}
  ]
  particleSpawnScript [&self:InitializeAnimTrail(Color:LinearColor=<rgba 1 1 1 0.2>, Lifetime=0.4, Ribbon Width Scale=1) {module=InitializeAnimTrail}]
  particleUpdateScript [
   &ParticleState() {module=UpdateAge}
   &self:UpdateAnimTrail() {module=UpdateAnimTrail}
  ]
  rendererClasses [&NiagaraRibbonRendererProperties]
  renderers [{rendererClass=&NiagaraRibbonRendererProperties}]
 }
 Wyeth_AnimTrails001 {
  emitterUpdateScript [
   &EmitterState()
   &self:SpawnAnimTrail(
    AnimTrail is Active:NiagaraBool=&Boolean_AndOperation(
     Bool1=$User.bisFullyDeployed
     Bool2=&SetBoolByFloatComparison(A=$User.ForwardDot, B=0.1, Comparison Type=::NewEnumerator2)
    )
    Mesh:NiagaraDataInterfaceSkeletalMesh=$User.Glider
    Socket A=4
    Socket B=5
   ) {module=SpawnAnimTrail001}
  ]
  particleSpawnScript [&self:InitializeAnimTrail(Color:LinearColor=<rgba 1 1 1 0.2>, Lifetime=0.4, Ribbon Width Scale=1) {module=InitializeAnimTrail}]
  particleUpdateScript [
   &ParticleState() {module=UpdateAge}
   &self:UpdateAnimTrail() {module=UpdateAnimTrail}
  ]
  rendererClasses [&NiagaraRibbonRendererProperties]
  renderers [{rendererClass=&NiagaraRibbonRendererProperties}]
 }
 Wyeth_AnimTrails002 {
  emitterUpdateScript [
   &EmitterState()
   &self:SpawnAnimTrail(
    AnimTrail is Active:NiagaraBool=&Boolean_AndOperation(
     Bool1=$User.bisFullyDeployed
     Bool2=&SetBoolByFloatComparison(A=$User.ForwardDot, B=0.1, Comparison Type=::NewEnumerator2)
    )
    Mesh:NiagaraDataInterfaceSkeletalMesh=$User.Glider
    Socket A=2
    Socket B=3
   ) {module=SpawnAnimTrail002}
  ]
  particleSpawnScript [&self:InitializeAnimTrail(Color:LinearColor=<rgba 1 1 1 0.2>, Lifetime=0.4, Ribbon Width Scale=1) {module=InitializeAnimTrail}]
  particleUpdateScript [
   &ParticleState() {module=UpdateAge}
   &self:UpdateAnimTrail() {module=UpdateAnimTrail}
  ]
  rendererClasses [&NiagaraRibbonRendererProperties]
  renderers [{rendererClass=&NiagaraRibbonRendererProperties}]
 }
 Wyeth_AnimTrails003 {
  emitterUpdateScript [
   &EmitterState()
   &self:SpawnAnimTrail(
    AnimTrail is Active:NiagaraBool=&Boolean_AndOperation(
     Bool1=$User.bisFullyDeployed
     Bool2=&SetBoolByFloatComparison(A=$User.ForwardDot, B=0.1, Comparison Type=::NewEnumerator2)
    )
    Mesh:NiagaraDataInterfaceSkeletalMesh=$User.Glider
    Socket A=6
    Socket B=7
   ) {module=SpawnAnimTrail003}
  ]
  particleSpawnScript [&self:InitializeAnimTrail(Color:LinearColor=<rgba 1 1 1 0.2>, Lifetime=0.4, Ribbon Width Scale=0.75) {module=InitializeAnimTrail}]
  particleUpdateScript [
   &ParticleState() {module=UpdateAge}
   &self:UpdateAnimTrail() {module=UpdateAnimTrail}
  ]
  rendererClasses [&NiagaraRibbonRendererProperties]
  renderers [{rendererClass=&NiagaraRibbonRendererProperties}]
 }
}
summary {
 systemName=NS_Glider_Carbide_Black
 emitters [
  {emitterName=Wyeth_AnimTrails, rendererClasses [&NiagaraRibbonRendererProperties]}
  {emitterName=Wyeth_AnimTrails001, rendererClasses [&NiagaraRibbonRendererProperties]}
  {emitterName=Wyeth_AnimTrails002, rendererClasses [&NiagaraRibbonRendererProperties]}
  {emitterName=Wyeth_AnimTrails003, rendererClasses [&NiagaraRibbonRendererProperties]}
  {emitterName=Flare, rendererClasses [&NiagaraSpriteRendererProperties]}
  {emitterName=Lights, rendererClasses [&NiagaraLightRendererProperties]}
 ]
 userVariables [
  User.bisFrontEndPreview {type=NiagaraBool}
  User.bisFullyDeployed {type=NiagaraBool}
  User.ForwardDot {type=NiagaraFloat, defaultValue=0, description=""}
  User.Glider {
   type=NiagaraDataInterfaceSkeletalMesh
   defaultValue {
    struct=NiagaraExt_VariableValue_DataInterface
    value {dataInterface=&self:NiagaraDataInterfaceSkeletalMesh_0}
   }
  }
 ]
}
